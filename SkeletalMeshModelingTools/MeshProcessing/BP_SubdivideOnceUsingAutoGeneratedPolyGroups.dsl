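; Blueprint mesh processor for skeletal mesh modeling tools that subdivides a dynamic mesh once using auto-generated PolyGroups.
; ProcessDynamicMesh computes PolyGroups on the default layer through polygon detection, recovering polygon faces from the triangle mesh.
; It then applies a single level of OpenSubdiv Catmull-Clark subdivision driven by those groups.
; The function always returns false.
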(fn ProcessDynamicMesh
  (params (TargetMesh :type Dynamic%20Mesh%20Object%20Reference))
  (bind _returnvalue (GeometryScript|PolyGroups|ComputePolyGroupsFromPolygonDetection TargetMesh (struct :bDefaultLayer true :ExtendedLayerIndex 0) false false 1.0 1.0 1))
  (GeometryScript|OpenSubdiv|ApplyPolyGroupCatmullClarkSubD _returnvalue 1 (struct :bDefaultLayer true :ExtendedLayerIndex 0))
  (return false))
