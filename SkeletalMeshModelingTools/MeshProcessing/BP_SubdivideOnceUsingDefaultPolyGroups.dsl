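; Blueprint mesh processor for skeletal mesh modeling tools that subdivides a dynamic mesh once.
; ProcessDynamicMesh applies Geometry Script's OpenSubdiv PolyGroup Catmull-Clark subdivision at a single level, using the default PolyGroup layer.
; It modifies the target mesh in place and always returns false.
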(fn ProcessDynamicMesh
  (params (TargetMesh :type Dynamic%20Mesh%20Object%20Reference))
  (GeometryScript|OpenSubdiv|ApplyPolyGroupCatmullClarkSubD TargetMesh 1 (struct :bDefaultLayer true :ExtendedLayerIndex 0))
  (return false))
